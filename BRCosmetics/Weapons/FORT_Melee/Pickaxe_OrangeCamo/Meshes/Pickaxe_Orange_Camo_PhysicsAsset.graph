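# Physics asset for the Orange Camo two-handed pickaxe, giving its handle and two dyn_ bones simple capsule collision bodies.
# Constraints form a chain from Melee_TwoHanded_Handle through dyn_Axe_02 to dyn_Axe_line, whose zero-length capsule is effectively a small sphere.
# The handle-to-axe joint locks all swing and twist, while the axe-to-line joint leaves them unlocked with narrow limit angles, suggesting a small dangling part.
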PhysicsAsset /BRCosmetics/Weapons/FORT_Melee/Pickaxe_OrangeCamo/Meshes/Pickaxe_Orange_Camo_PhysicsAsset
bodies {
 Melee_TwoHanded_Handle [{shapeType=Capsule, length=83.7658, radius=4.7008, center=<0.6464 -2.9792 29.5843>, rotation=<rot -0.7951 -89.9999 0.0011>}]
 dyn_Axe_02 [{shapeType=Capsule, length=43.8864, radius=12.9491, center=<-0.4636 11.5198 -0.0026>, rotation=<rot 0.0007 -27.6893 -89.9847>}]
 dyn_Axe_line [{shapeType=Capsule, length=0, radius=1.5817, center=<-0.1794 0.0266 0.3012>, rotation=<rot 0.2889 -102.6432 -88.059>}]
}
constraints [
 {
  bone1Name=dyn_Axe_02
  bone2Name=Melee_TwoHanded_Handle
  swing1LimitDegrees=45
  swing1Motion=Locked
  swing2LimitDegrees=45
  swing2Motion=Locked
  twistLimitDegrees=45
  twistMotion=Locked
 }
 {bone1Name=dyn_Axe_line, bone2Name=dyn_Axe_02, swing1LimitDegrees=10, swing2LimitDegrees=15, twistLimitDegrees=10}
]
