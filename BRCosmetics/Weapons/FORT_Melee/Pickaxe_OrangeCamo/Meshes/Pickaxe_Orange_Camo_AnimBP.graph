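# Post-process style Anim Blueprint for the Orange Camo pickaxe cosmetic's skeleton, with an empty update event.
# The output pose is the input pose passed through a local-to-component and back conversion.
# An Anim Dynamics node for the dyn_Axe_line bone and a RigidBody node using the pickaxe's physics asset are configured for secondary motion.
# Neither physics node is wired into the output, so the pose passes through unchanged.
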AnimBlueprint /BRCosmetics/Weapons/FORT_Melee/Pickaxe_OrangeCamo/Meshes/Pickaxe_Orange_Camo_AnimBP
target_skeleton=&/Pickaxe_Orange_Camo_Skeleton
graphs [
 AnimGraph {nodes [
   Anim Dynamics - Bone: dyn_Axe_line(Alpha=1, ExternalForce="0,0,0") {
    angularSpringConstant=1500
    bGravityOverrideInSimSpace=true
    bUseGravityOverride=true
    alphaBoolBlend {blendOption=Linear}
    boundBone {boneName=dyn_Axe_line}
    componentAppliedLinearAccClamp=<2000 2000 2000>
    componentLinearAccScale=<1 1 1>
    componentLinearVelScale=<0.5 0.5 0.5>
    gravityOverride {z=-980}
    physicsBodyDefinitions [{
      sphereCollisionRadius=0
      boundBone {boneName=dyn_Axe_line}
      boxExtents=<10 10 10>
      constraintSetup {angularLimitsMax=<0 10 10>, angularLimitsMin=<0 -10 -10>, angularTarget=<0 0 0>, linearAxesMax=<0.2 0.2 0.2>, linearAxesMin=<-0.2 -0.2 -0.2>}
      localJointOffset=<1 -0.1 -1>
     }]
   }
   RigidBody() {
    bOverrideWorldGravity=true
    overridePhysicsAsset=&/Pickaxe_Orange_Camo_PhysicsAsset
    alphaBoolBlend {blendOption=Linear}
    componentAppliedLinearAccClamp=<1500 1500 1500>
    componentLinearAccScale=<1 1 1>
    componentLinearVelScale=<1 1 1>
    overrideWorldGravity {z=-980}
   }
   AnimGraphNode_SubInput_0 = Input Pose()
   Output Pose(Result=Component To Local(ComponentPose=Local To Component(LocalPose=@AnimGraphNode_SubInput_0))) {layerGroup=None}
  ]}
 EventGraph {nodes [Event BlueprintUpdateAnimation()]}
]
variables []
